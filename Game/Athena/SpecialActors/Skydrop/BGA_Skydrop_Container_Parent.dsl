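; Parent Blueprint for Skydrop loot containers, interactable once and marked by a looping sound indicator.
; Interacting plays the open cue and stops the indicator; the server rolls the loot tier group, then destroys the container.
; A single drop goes to the player's inventory if a slot is free, re-rolling the table for that item; otherwise loot spawns on the floor.
; Death before searching spills floor loot with a replicated destroy cue.
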(event EventBeginPlay
  (SoundIndicator|StartSoundIndicatorLooping (Variables|Default|GetFortSoundIndicator) (Transformation|GetActorLocation self) (Variables|/Script/Engine.AudioComponent|GetSound (Variables|Default|GetAmbience)) 1.0 0)
  (|CallParentFunction))

(event Interaction|EventBlueprintOnInteract
  (params (InteractingPawn :type Fort%20Pawn%20Object%20Reference) (InteractionBeingAttempted :type EInteractionBeingAttempted%20Enum :default "FirstInteraction"))
  (if (not (Variables|Default|GetSearched))
    (Variables|Default|SetSearched true)
    (bind _output_get (Variables|Default|SetLastInteractingPawn InteractingPawn))
    (CallFunction|OpenContainer _output_get)))

(event Custom|OpenContainer
  (params (Target :type Pawn%20Object%20Reference))
  (CallFunction|SpawnLootLogic)
  (CallFunction|PlayOpenFX)
  (SoundIndicator|StopSoundIndicatorLooping (Variables|Default|GetFortSoundIndicator)))

(event Custom|SpawnLootLogic
  (if (Networking|IsServer)
    (bind (_outloottodrop _returnvalue) (Loot|PickLootDrops (Variables|Default|GetLootTierGroupName) 0 0 "(GameplayTags=)"))
    (if _returnvalue
      (Collision|SetActorEnableCollision false)
      (Utilities|FlowControl|DelayUntilNextTick)
      (bind _returnvalue_2 (CallFunction|HasOpenInventorySlot (Variables|Default|GetLastInteractingPawn)))
      (if (and _returnvalue_2 (== (Utilities|Array|Length _outloottodrop) 1))
        (CallFunction|GiveLootItemDirectlytoPlayer 0)
        (else
          (CallFunction|SpawnFloorLootOnly)))
      (Utilities|FlowControl|Delay 0.1)
      (CallFunction|DestroyContainer)
      (else
        (CallFunction|DestroyContainer)))))

(event Custom|GiveLootItemDirectlyToPlayer
  (params (Target :type Pawn%20Object%20Reference))
  (Development|PrintString "Granting item")
  (bind _asfort_player_controller (Utilities|Casting|CastToFortPlayerController (Pawn|GetController Target)))
  (bind _item_definition (CallFunction|GetItemtoGrant))
  (bind _asfort_world_item_definition (Utilities|Casting|CastToFortWorldItemDefinition _item_definition))
  (Development|PrintString (Utilities|GetDisplayName _asfort_world_item_definition))
  (Inventory|GiveItemtoInventoryOwner _asfort_player_controller _asfort_world_item_definition))

(event Custom|SpawnFloorLootOnly
  (Class|AthenaFunctionLibrary|SpawnLootfromTable (Variables|Default|GetLootTierGroupName) 1 -1 (Math|Transform|TransformLocation (Transformation|GetActorTransform) (Variables|Default|GetFloorLootSpawnLocOffset)) (- (Transformation|GetActorLocation) (Variables|Default|GetFloorLootSpawnLocOffset)) 0 true))

(event EventOnDeathServer
  (params (Damage :type Float%20%28single-precision%29) (DamageTags :type Gameplay%20Tag%20Container%20Structure%20%28by%20ref%29) (Momentum :type Vector :default "0, 0, 0") (HitInfo :type Hit%20Result%20Structure%20%28by%20ref%29) (InstigatedBy :type Controller%20Object%20Reference) (DamageCauser :type Actor%20Object%20Reference) (EffectContext :type Gameplay%20Effect%20Context%20Handle%20Structure))
  (if (not (Variables|Default|GetSearched))
    (CallFunction|SpawnFloorLootOnly)
    (Ability|ExecuteGameplayCueWithParamsReplicated (Variables|Default|GetGC_Destroy) (Ability|GameplayCue|MakeGameplayCueParameters :NormalizedMagnitude 0.0 :RawMagnitude 0.0 :Location (Transformation|GetActorLocation) :Normal "0,0,0" :GameplayEffectLevel 1 :AbilityLevel 1 :bReplicateLocationWhenUsingMinimalRepProxy false))))

(event EventOnDeathPlayEffects
  (params (Damage :type Float%20%28single-precision%29) (DamageTags :type Gameplay%20Tag%20Container%20Structure%20%28by%20ref%29) (Momentum :type Vector :default "0, 0, 0") (HitInfo :type Hit%20Result%20Structure%20%28by%20ref%29) (InstigatedBy :type Fort%20Pawn%20Object%20Reference) (DamageCauser :type Actor%20Object%20Reference) (EffectContext :type Gameplay%20Effect%20Context%20Handle%20Structure))
  (Development|SetHiddeninGame (Variables|Default|GetStaticMesh) true true)
  (SoundIndicator|StopSoundIndicatorLooping (Variables|Default|GetFortSoundIndicator)))

(event Custom|DestroyContainer
  (Components|Activation|Deactivate (Variables|Default|GetAthenaSpecialActor))
  (Actor|DestroyActor))

(fn BlueprintCanInteract
  (params (InteractingPawn :type Fort%20Pawn%20Object%20Reference) (InteractionBeingAttempted :type EInteractionBeingAttempted%20Enum :default "FirstInteraction") (InteractionType :type TInteractionType%20Enum :default "IT_NoInteraction"))
  (return (not (Variables|Default|GetSearched))))

(fn BlueprintGetInteractionString
  (params (InteractingPawn :type Fort%20Pawn%20Object%20Reference) (InteractionBeingAttempted :type EInteractionBeingAttempted%20Enum :default "FirstInteraction"))
  (return (Variables|Default|GetInteractMessage)))

(fn BlueprintOnLocalInteract
  (params (InteractingPawn :type Fort%20Player%20Pawn%20Object%20Reference) (InteractionBeingAttempted :type EInteractionBeingAttempted%20Enum :default "FirstInteraction"))
  (return false))

(fn GetItemToGrant
  (bind _outloottodrop (Loot|PickLootDrops (Variables|Default|GetLootTierGroupName) 0 0 "(GameplayTags=)"))
  (Development|PrintString (Utilities|GetDisplayName (Class|FortItemFunctionLibrary|BreakFortItemEntry (Utilities|Array|Get(aref) _outloottodrop))))
  (return (Class|FortItemFunctionLibrary|BreakFortItemEntry (Utilities|Array|Get(aref) _outloottodrop))))

(fn HasOpenInventorySlot
  (params (Target :type Pawn%20Object%20Reference))
  (bind _asfort_player_controller (Utilities|Casting|CastToFortPlayerController (Pawn|GetController Target))
    (:then
      (bind _returnvalue (IndexSelectionModel|GetNumItems(Message) _asfort_player_controller))
      (Development|PrintString (Utilities|String|ToString(Integer) _returnvalue))
      (return (< _returnvalue 5)))
    (:CastFailed
      (return false))))

(fn PlayOpenFX
  (GameplayCue|ExecuteGameplayCue%28Burst%29Local self (Variables|Default|GetGC_Open) (Ability|GameplayCue|MakeGameplayCueParameters :NormalizedMagnitude 0.0 :RawMagnitude 0.0 :Location (Transformation|GetActorLocation) :Normal "0,0,0" :Instigator self :GameplayEffectLevel 1 :AbilityLevel 1 :bReplicateLocationWhenUsingMinimalRepProxy false)))

(fn ConstructionScript
  (|CallParentFunction))
