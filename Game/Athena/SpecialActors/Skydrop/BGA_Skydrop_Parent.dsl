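; Parent Blueprint for a skydrop pod that falls at an angle after a delay, using server-driven projectile movement tuned by ScalableFloat rows.
; Server hit logic destroys weaker buildings, launches characters and vehicles, deflects once off steep indestructible surfaces, and otherwise lands.
; Landing plays a gameplay cue, spawns the configured container, calls an overridable hook, then kills the pod.
; Destruction before landing can spawn the container's loot instead.
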(event Custom|ResumeMovementDownward
  (params (HitNormal :type Vector))
  (Utilities|FlowControl|DelayUntilNextTick)
  (Components|Movement|SetUpdatedComponent (Variables|Default|GetFortProjectileMovement) (Variables|Default|GetSphere))
  (Variables|Default|SetDidBounce true)
  (CallFunction|SetMovementVelocity (* HitNormal 1e+03))
  (Transformation|SetActorRotation "-90.000000, 0, 0")
  (Variables|/Script/Engine.ProjectileMovementComponent|SetProjectileGravityScale 1.0 (Variables|Default|GetFortProjectileMovement)))

(event EventBeginPlay
  (CallFunction|Initialize)
  (if (Variables|Default|GetDebugStartGrounded)
    (Variables|Default|SetisGrounded true)
    (else
      (Utilities|Time|SetTimerbyEvent (EventDispatchers|CreateEvent :!event BeginDrop) (ScalableFloat|GetValueatLevel (Variables|Default|GetRow_DropPredelay))))))

(event Custom|OnLandedClient
  (Networking|CanExecuteCosmeticEvents
    (:True
      (CallFunction|PlayLandedFX))))

(event Custom|ResumeMovement
  ;[1 1 1] Resume Movement after a false hit
  (Utilities|FlowControl|DelayUntilNextTick)
  (Components|Movement|SetUpdatedComponent (Variables|Default|GetFortProjectileMovement) (Variables|Default|GetSphere))
  (CallFunction|SetMovementVelocity (* (Variables|Default|GetfallDirection) (ScalableFloat|GetValueatLevel (Variables|Default|GetRow_FallVelocity)))))

(event Custom|SetMovementVelocity
  (params (InVelocity :type Vector))
  ;[1 1 1] Set the velocity of the Projectile Movement Component. Only call on server, will be replicated to client
  (Projectile|SetVelocity (Variables|Default|GetFortProjectileMovement) InVelocity))

(event Collision|EventHit
  (params (MyComp :type Primitive%20Component%20Object%20Reference) (Other :type Actor%20Object%20Reference) (OtherComp :type Primitive%20Component%20Object%20Reference) (bSelfMoved :type Boolean) (HitLocation :type Vector :default "0, 0, 0") (HitNormal :type Vector :default "0, 0, 0") (NormalImpulse :type Vector :default "0, 0, 0") (Hit :type Hit%20Result%20Structure%20%28by%20ref%29))
  (bind _hitactor (Variables|Default|GethitActor))
  (bind _hitactor_2 (Variables|Default|GethitActor))
  ;[1 1 1 1 30] Hit Impact Logic
  (bind _output_get (Variables|Default|SethitActor Other))
  (if (Networking|IsServer)
    (Development|PrintString (Utilities|String|Append "drop pod hit: " (Utilities|GetDisplayName _output_get)))
    (Utilities|IsValid _hitactor
      (:Is%20Valid
        (if (and (Variables|/Script/Engine.Actor|GetbCanBeDamaged _hitactor) (ScalableFloat|GetValueasBool (Variables|Default|GetRow_DestroysObstacles)))
          (bind _asbuilding_actor (Utilities|Casting|CastToBuildingActor (Variables|Default|GethitActor))
            (:then
              ;;[1 0.4908 0.4609] Hit Building; attempt to destroy and continue moving
              (if (and (> (Health|GetMaxHealth _asbuilding_actor) 0.0) (<= (Health|GetMaxHealth _asbuilding_actor) 5e+03))
                (CallFunction|DestroyBuilding _asbuilding_actor)
                (if (<= (Health|GetHealth _asbuilding_actor) 0)
                  (Development|PrintString "destroyed obscacle; continuing")
                  (CallFunction|ResumeMovement)
                  (else
                    (Development|PrintString "Failed to destroy obstacle; landing")))
                (else
                  (Development|PrintString "Hit something non-viable; landing")))
              (CallFunction|OnLanded))
            (:CastFailed
              ;;[1 0.4524 0.2474] Hit a destructible non-building actor like a player or vehicle; continue moving
              (Development|PrintString "Hit a destructible, non-building actor ")
              (Collision|IgnoreActorwhenMoving (Variables|Default|GetSphere) _hitactor_2 true)
              (CallFunction|LaunchCharacters&Vehicles _hitactor_2)
              (CallFunction|ResumeMovement)))
          (else
            ;;[1 0.6685 0.2208] Hit indestructible; land
            (bind _returnvalue (Fortnite|GetAngleDegrees HitNormal (Math|Vector|MakeVector 0.0 0.0 1.0)))
            (if (Variables|Default|GetDidBounce)
              (CallFunction|OnLanded)
              (elif (> _returnvalue 45.0)
                (CallFunction|ResumeMovementDownward HitNormal)
                (else
                  (CallFunction|OnLanded))))))))))

(event Custom|LaunchCharacters&Vehicles
  (params (HitActor :type Actor%20Object%20Reference))
  (bind _returnvalue (Transformation|GetActorLocation HitActor))
  (bind _returnvalue_2 (Transformation|GetActorLocation))
  (bind _returnvalue_3 (Math|Vector|GetUnitDirection(Vector) _returnvalue_2 _returnvalue))
  (bind _returnvalue_4 (Math|Vector|Normalize _returnvalue_3))
  (bind _row_playercollisionlaunchvel (Variables|Default|GetRow_PlayerCollisionLaunchVel))
  (bind _returnvalue_5 (ScalableFloat|GetValueatLevel _row_playercollisionlaunchvel))
  (bind _returnvalue_6 (* _returnvalue_4 _returnvalue_5))
  (bind _x (.x _returnvalue_6))
  (bind _y (.y _returnvalue_6))
  (bind _returnvalue_7 (Math|Vector|MakeVector _x _y _returnvalue_5))
  ;[1 1 1 1 30] Hit Impact Logic
  ;;[1 1 1] Launch characters and vehicles if we run into them
  (Utilities|FlowControl|DoOnce false
    (:Completed
      (Utilities|FlowControl|Sequence
        (:then_0
          _)
        (:then_1
          (Utilities|FlowControl|Delay 0.2)))
      (bind _asfort_pawn (Utilities|Casting|CastToFortPawn HitActor)
        (:then
          (Character|LaunchCharacter _asfort_pawn _returnvalue_7 false true))
        (:CastFailed
          (bind _asfort_vehicle_interface (Utilities|Casting|CastToFortVehicleInterface HitActor))
          (Audio|BroadcastAbilityImpactatLocation 0 (Utilities|ToObject(Interface) _asfort_vehicle_interface) _returnvalue_7 777.0) ; This launches vehicles (named weird for anti-cheat reasons)\nPosition = velocity to launch it at\nRadius must equal 777\nGates knows more about it
        )))) ; TODO make this once per specific actor instead of using delay
)

(event EventOnDeathServer
  (params (Damage :type Float%20%28single-precision%29) (DamageTags :type Gameplay%20Tag%20Container%20Structure%20%28by%20ref%29) (Momentum :type Vector :default "0, 0, 0") (HitInfo :type Hit%20Result%20Structure%20%28by%20ref%29) (InstigatedBy :type Controller%20Object%20Reference) (DamageCauser :type Actor%20Object%20Reference) (EffectContext :type Gameplay%20Effect%20Context%20Handle%20Structure))
  (|CallParentFunction Damage DamageTags Momentum HitInfo InstigatedBy DamageCauser EffectContext)
  (if (and (not (Variables|Default|GetisGrounded)) (ScalableFloat|GetValueasBool (Variables|Default|GetRow_SpawnLootOnEarlyDestruction)))
    (bind _asbga_skydrop_container_parent (Utilities|Casting|CastToBGA_Skydrop_Container_ParentClass (Variables|Default|GetContainerToSpawnOnLanding)))
    (Development|PrintString "Drop pod destroyed before landing, attempting to spawn container loot contents")
    (Class|AthenaFunctionLibrary|SpawnLootfromTable (.LootTierGroupName (Class|GetClassDefaults _asbga_skydrop_container_parent)) 1 -1 (Math|Transform|TransformLocation (Transformation|GetActorTransform) (.FloorLootSpawnLocOffset (Class|GetClassDefaults _asbga_skydrop_container_parent))) (- (Transformation|GetActorLocation) (.FloorLootSpawnLocOffset (Class|GetClassDefaults _asbga_skydrop_container_parent))) 0 true)))

(fn BeginDrop
  (if (Networking|IsServer)
    (if (Utilities|Operators|Equal(==) :A (Variables|Default|GetfallDirection))
      (Variables|Default|SetfallDirection (Math|Vector|RotateVectorAroundAxis (Transformation|GetActorForwardVector) (Math|Random|RandomFloatinRange (ScalableFloat|GetValueatLevel (Variables|Default|GetRow_FallAngleMin)) (ScalableFloat|GetValueatLevel (Variables|Default|GetRow_FallAngleMax))) (Transformation|GetActorRightVector))))
    (CallFunction|SetMovementVelocity (* (Variables|Default|GetfallDirection) (ScalableFloat|GetValueatLevel (Variables|Default|GetRow_FallVelocity))))))

(fn DestroyBuilding
  (params (Actor :type Actor%20Object%20Reference))
  (GameplayEffects|ApplyGameplayEffectToTarget :self (Ability|GetAbilitySystemComponent self) :GameplayEffectClass "/Game/Athena/SpecialActors/Skydrop/GE_Skydrop_Environment_Destruction.GE_Skydrop_Environment_Destruction_C" :Target (Ability|GetAbilitySystemComponent Actor) :Level 0.0))

(fn Initialize
  ;[1 1 1] Overridable function to perform any startup initialization
  (Components|Sphere|SetSphereRadius (Variables|Default|GetSphere) (ScalableFloat|GetValueatLevel (Variables|Default|GetRow_CollisionRadius)) false))

(fn OnLanded
  (bind (_returnvalue_location _returnvalue_rotation) (Math|Transform|BreakTransform (Transformation|GetActorTransform)))
  (Development|PrintString "Pod Landed")
  (if (not (Variables|Default|GetisGrounded))
    (CallFunction|SetMovementVelocity "0, 0, 0")
    (Components|Movement|StopMovementImmediately (Variables|Default|GetFortProjectileMovement))
    (Variables|Default|SetisGrounded true)
    (CallFunction|PlayLandedFX)
    (CallFunction|OnLandedClient)
    (if (Networking|IsServer)
      (bind _containertospawnonlanding (Variables|Default|GetContainerToSpawnOnLanding)
        (:then
          (if (Variables|Default|GetSpawnUpright)
            (Variables|Default|SetSpawnRotation "0, 0, 0")
            (else
              (Variables|Default|SetSpawnRotation _returnvalue_rotation)))
          (bind _returnvalue (Game|SpawnActorfromClass _containertospawnonlanding (Math|Transform|MakeTransform _returnvalue_location (Variables|Default|GetSpawnRotation)) "Undefined" "SelectDefaultAtRuntime" 0 0 0)))
        (:else
          _))
      (CallFunction|OnSpawnedContainer _returnvalue)
      (Collision|SetActorEnableCollision false)
      (Rendering|SetActorHiddenInGame true)
      (Gameplay|Die))))

(fn OnSpawnedContainer
  (params (Container :type Actor%20Object%20Reference))
  (return false))

(fn PlayLandedFX
  (Ability|ExecuteGameplayCueWithParamsReplicated (Variables|Default|GetGC_Landed) (Ability|GameplayCue|MakeGameplayCueParameters :NormalizedMagnitude 0.0 :RawMagnitude 0.0 :Location (Transformation|GetActorLocation) :Normal "0,0,1.000000" :GameplayEffectLevel 1 :AbilityLevel 1 :bReplicateLocationWhenUsingMinimalRepProxy false)))

(fn ConstructionScript)
